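; Athena Paprika sniper weapon Blueprint that adds a local camera shake when fired while targeting with iron sights.
; BeginPlay assigns a mod-added handler that loops over weapon mod slots, comparing tuning tags to update the HasIronSights flag.
; Fire FX starts the shake on the local viewing player's camera when that flag is set and the weapon is targeting.
; Targeting changes delay the parent call one tick because the scoped setup otherwise misreports first-person camera use.
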(event EventBeginPlay
  ;[1 1 1] BeginPlay: If we have Iron Sights, set Bool to True, which will enable camera shake when fired while ADS in 3P
  (|CallParentFunction)
  (WeaponMods|AssignOnModAddedtoWeapon (delegate Custom|OnModAddedToWeapon_Event)))

(event Custom|OnModAddedToWeapon_Event
  (params (Mod :type Fort%20Weapon%20Mod%20Item%20Definition%20Object%20Reference) (Weapon :type Fort%20Weapon%20Object%20Reference))
  ;[1 1 1] BeginPlay: If we have Iron Sights, set Bool to True, which will enable camera shake when fired while ADS in 3P
  (bind _array_element (Utilities|Array|ForEachLoopwithBreak (CallFunction|GetWeaponModSlots))
    (:LoopBody
      (if (== (Variables|/Script/FortniteGame.FortWeaponModItemDefinition|GetPluginTuningTag (Utilities|Struct|BreakFortWeaponModSlot _array_element)) (Variables|Default|GetIronSightsModTag))
        (Variables|Default|SetHasIronSights true))
      (Variables|Default|SetHasIronSights false))))

(event EventOnPlayWeaponFireFX
  (params (bPersistentFire :type Boolean) (bSecondaryFire :type Boolean))
  ;[1 1 1] PlayWeaponFireFX: Play Camera Shake if we're ADS with Iron Sights
  (|CallParentFunction bPersistentFire bSecondaryFire)
  (if (and (Variables|Default|GetHasIronSights) (FortWeapon|IsTargeting))
    (CameraShakes|StartCameraShake :self (Variables|/Script/Engine.PlayerController|GetPlayerCameraManager (Pawn|GetLocalViewingPlayerController (Utilities|Casting|CastToFortPlayerPawn (Game|GetInstigator)))) :ShakeClass (Variables|Default|GetFireCameraShake) :Scale 1.0 :PlaySpace "CameraLocal")))

(event EventOnSetTargeting
  (params (bNewIsTargeting :type Boolean))
  ;[1 1 1] Weapon uses a slightly different setup to support scoped targeting, causing the evaluation of 'bIsCurrentlyTargeting' in FortRangedWeapon::IsUsingFirstPersonCamera (which calls FortPlayerController::IsUsingFirstPersonCamera) to be 'true' even if we're no longer targeting.
  (Utilities|FlowControl|DelayUntilNextTick)
  (|CallParentFunction bNewIsTargeting))

(fn ConstructionScript
  (|CallParentFunction))
