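# Sound cue that plays when the Paprika sniper weapon is dropped.
# A single drop wave feeds a Modulator at default settings for slight pitch and volume variation.
# It plays at reduced volume through the loot-collect sound class with shared weapon-drop attenuation.
# It references the shared Weapon_Drop_StopOldest concurrency set to limit overlapping drop sounds.
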SoundCue /PaprikaCoreWeapons/Items/Weapons/PaprikaSniper/Sounds/SC_Paprika_Sniper_Drop
root=Modulator_1
nodes [
 {id=Modulator_1, children [WavePlayer_6]}
 {id=WavePlayer_6, properties {soundWaveAssetPtr=/PaprikaCoreWeapons/Items/Weapons/PaprikaSniper/Sounds/Waves/Sniper_Paprika_Drop_01}}
]
properties {
 attenuationSettings=&Fort_Weapon_Drop_Att
 maxDistance=3000
 soundClassObject=&Fort_Loot_Collect
 volumeMultiplier=0.6
 concurrencyOverrides {resolutionRule=PreventNew}
 concurrencySet [&Weapon_Drop_StopOldest]
}
