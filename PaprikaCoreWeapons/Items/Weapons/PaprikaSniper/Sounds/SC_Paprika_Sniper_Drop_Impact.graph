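# Sound cue for the Paprika sniper rifle's drop impact, played as loot-collection class audio when the weapon lands.
# A single landing wave passes through a modulator, with slightly lowered pitch and volume.
# Attenuation and concurrency are overridden inline to keep it short-range and cap simultaneous instances, preventing new ones once the limit is reached.
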SoundCue /PaprikaCoreWeapons/Items/Weapons/PaprikaSniper/Sounds/SC_Paprika_Sniper_Drop_Impact
root=Modulator_0
nodes [
 {id=Modulator_0, children [WavePlayer_3]}
 {id=WavePlayer_3, properties {soundWaveAssetPtr=/PaprikaCoreWeapons/Items/Weapons/PaprikaSniper/Sounds/Waves/Sniper_Paprika_Land_01}}
]
properties {
 bOverrideAttenuation=true
 bOverrideConcurrency=true
 maxDistance=1000
 pitchMultiplier=0.95
 soundClassObject=&Fort_Loot_Collect
 volumeMultiplier=0.8
 attenuationOverrides {falloffDistance=1000, attenuationShapeExtents {x=0}}
 concurrencyOverrides {maxCount=3, resolutionRule=PreventNew}
}
