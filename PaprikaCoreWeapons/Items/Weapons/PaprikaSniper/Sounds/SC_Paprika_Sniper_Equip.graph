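# Sound cue for the Paprika sniper rifle's equip sound, in the weapon foley sound class.
# Two equally weighted wave variations feed a random node and a shared modulator that varies pitch and volume.
# An IsLocal branch sends this through an attenuation node without assigned settings when true, or through the shared Fortnite weapon reload attenuation when false or unset.
# The cue overrides concurrency and has its debug flag enabled.
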SoundCue /PaprikaCoreWeapons/Items/Weapons/PaprikaSniper/Sounds/SC_Paprika_Sniper_Equip
root=Branch_0
nodes [
 {id=Attenuation_0, children [Modulator_0]}
 {id=Attenuation_1, children [Modulator_0], properties {attenuationSettings=&Fort_Weapon_Reload_Att}}
 {id=Branch_0, children [Attenuation_0, Attenuation_1, Attenuation_1], properties {boolParameterName=IsLocal}}
 {id=Modulator_0, children [Random_0], properties {pitchMax=1, volumeMax=1}}
 {id=Random_0, children [WavePlayer_0,WavePlayer_1], properties {weights [1,1]}}
 {id=WavePlayer_0, properties {soundWaveAssetPtr=/PaprikaCoreWeapons/Items/Weapons/PaprikaSniper/Sounds/Waves/Sniper_Paprika_Equip_01}}
 {id=WavePlayer_1, properties {soundWaveAssetPtr=/PaprikaCoreWeapons/Items/Weapons/PaprikaSniper/Sounds/Waves/Sniper_Paprika_Equip_02}}
]
properties {
 bDebug=true
 bOverrideConcurrency=true
 maxDistance=2097152
 soundClassObject=&Fort_WeaponFoley
 volumeMultiplier=0.677
}
