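# Sound cue for the cartridge eject during the Paprika Sniper reload.
# A random node picks one of three equally weighted suppressed-sniper cartridge eject waves.
# It plays at reduced volume in the weapon foley sound class, using the shared weapon reload attenuation and overriding concurrency, with debug enabled.
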SoundCue /PaprikaCoreWeapons/Items/Weapons/PaprikaSniper/Sounds/SC_Paprika_Sniper_Reload_Cartridge_Eject
root=Random_0
nodes [
 {id=Random_0, children [WavePlayer_1,WavePlayer_2,WavePlayer_3], properties {weights [1,1,1]}}
 {
  id=WavePlayer_1
  properties {soundWaveAssetPtr=/PaprikaCoreWeapons/Items/Weapons/PaprikaSniper/Sounds/Waves/SniperSuppressed_Reload_CartridgeEject_01}
 }
 {
  id=WavePlayer_2
  properties {soundWaveAssetPtr=/PaprikaCoreWeapons/Items/Weapons/PaprikaSniper/Sounds/Waves/SniperSuppressed_Reload_CartridgeEject_02}
 }
 {
  id=WavePlayer_3
  properties {soundWaveAssetPtr=/PaprikaCoreWeapons/Items/Weapons/PaprikaSniper/Sounds/Waves/SniperSuppressed_Reload_CartridgeEject_03}
 }
]
properties {
 attenuationSettings=&Fort_Weapon_Reload_Att
 bDebug=true
 bOverrideConcurrency=true
 maxDistance=2250
 soundClassObject=&Fort_WeaponFoley
 volumeMultiplier=0.4
}
