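# Sound cue for the cartridge impact heard during the Paprika sniper's reload.
# A random node picks one of three equally weighted cartridge-impact waves, borrowed by name from a core AR set, and feeds a modulator that varies pitch slightly downward.
# Cue-level multipliers raise pitch and lower volume, with weapon foley sound class, shared reload attenuation, and a concurrency override.
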SoundCue /PaprikaCoreWeapons/Items/Weapons/PaprikaSniper/Sounds/SC_Paprika_Sniper_Reload_Cartridge_Impact
root=Modulator_0
nodes [
 {id=Modulator_0, children [Random_1], properties {pitchMax=0.95, pitchMin=0.85}}
 {id=Random_1, children [WavePlayer_1,WavePlayer_2,WavePlayer_3], properties {weights [1,1,1]}}
 {
  id=WavePlayer_1
  properties {soundWaveAssetPtr=/PaprikaCoreWeapons/Items/Weapons/PaprikaSniper/Sounds/Waves/Radical_CoreAR_Athena_CartridgeImpact_03}
 }
 {
  id=WavePlayer_2
  properties {soundWaveAssetPtr=/PaprikaCoreWeapons/Items/Weapons/PaprikaSniper/Sounds/Waves/Radical_CoreAR_Athena_CartridgeImpact_02}
 }
 {
  id=WavePlayer_3
  properties {soundWaveAssetPtr=/PaprikaCoreWeapons/Items/Weapons/PaprikaSniper/Sounds/Waves/Radical_CoreAR_Athena_CartridgeImpact_01}
 }
]
properties {
 attenuationSettings=&Fort_Weapon_Reload_Att
 bDebug=true
 bOverrideConcurrency=true
 maxDistance=2250
 pitchMultiplier=1.2
 soundClassObject=&Fort_WeaponFoley
 volumeMultiplier=0.2
}
