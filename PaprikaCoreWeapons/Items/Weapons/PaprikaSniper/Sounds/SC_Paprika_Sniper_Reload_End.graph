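# Sound cue for the end of the Paprika sniper's reload, classed as weapon foley.
# A single wave player feeds a mixer that boosts its volume above unity.
# It uses the shared weapon reload attenuation, overrides concurrency at the cue level and has its debug flag enabled.
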SoundCue /PaprikaCoreWeapons/Items/Weapons/PaprikaSniper/Sounds/SC_Paprika_Sniper_Reload_End
root=Mixer_0
nodes [
 {id=Mixer_0, children [WavePlayer_0], properties {inputVolume [1.85]}}
 {
  id=WavePlayer_0
  properties {soundWaveAssetPtr=/PaprikaCoreWeapons/Items/Weapons/PaprikaSniper/Sounds/Waves/Weapon_Sniper_Paprika_Reload_End}
 }
]
properties {
 attenuationSettings=&Fort_Weapon_Reload_Att
 bDebug=true
 bOverrideConcurrency=true
 maxDistance=2250
 soundClassObject=&Fort_WeaponFoley
 volumeMultiplier=1
}
