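# Sound cue for a cartridge impact during the Paprika sniper's reload, using suppressed-sniper reload waves.
# A Random node picks one of three equally weighted impact variations and feeds a Modulator at the root.
# It plays quietly in the weapon foley sound class with weapon-reload attenuation, overrides concurrency, and has the debug flag enabled.
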SoundCue /PaprikaCoreWeapons/Items/Weapons/PaprikaSniper/Sounds/SC_Paprika_Sniper_Reload_Shell_Impact
root=Modulator_0
nodes [
 {id=Modulator_0, children [Random_1]}
 {id=Random_1, children [WavePlayer_0,WavePlayer_4,WavePlayer_5], properties {weights [1,1,1]}}
 {
  id=WavePlayer_0
  properties {soundWaveAssetPtr=/PaprikaCoreWeapons/Items/Weapons/PaprikaSniper/Sounds/Waves/SniperSuppressed_Reload_CartridgeImpact_01}
 }
 {
  id=WavePlayer_4
  properties {soundWaveAssetPtr=/PaprikaCoreWeapons/Items/Weapons/PaprikaSniper/Sounds/Waves/SniperSuppressed_Reload_CartridgeImpact_02}
 }
 {
  id=WavePlayer_5
  properties {soundWaveAssetPtr=/PaprikaCoreWeapons/Items/Weapons/PaprikaSniper/Sounds/Waves/SniperSuppressed_Reload_CartridgeImpact_03}
 }
]
properties {
 attenuationSettings=&Fort_Weapon_Reload_Att
 bDebug=true
 bOverrideConcurrency=true
 maxDistance=2250
 soundClassObject=&Fort_WeaponFoley
 volumeMultiplier=0.1
}
